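; Movement mode extension logic for wingsuit flight that manages camera modifiers and landing feedback.
; Activation rebuilds the camera modifier on non-authority local viewing controllers, using separate owner and spectator classes, and cancels pending removal.
; Deactivation applies a holster effect, sends WingSuit.End, disables the camera and removes it after a short timer.
; Landing mode and input select a roll, vertical or knee-slide montage from runtime data.
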(event Custom|HandleIsLocallyViewedChanged
  (params (IsLocallyViewed :type Boolean))
  ;[1 1 1] TODO: clean up spectators when they look away
  (if (not IsLocallyViewed)
    (CallFunction|SetCameraModifier false true)))

(event EventOnExtensionActivate
  (params (RuntimeContext :type MMERuntime%20Context%20Structure%20%28by%20ref%29))
  (CallFunction|SetCameraModifier false true)
  (CallFunction|SetCameraModifier true)
  (Utilities|Time|ClearandInvalidateTimerbyHandle (Variables|FortMovementMode_BaseExtLogic|GetOwnerPawn) (Variables|Default|GetTimerToDestroy)))

(event EventOnExtensionDeactivate
  (params (RuntimeContext :type MMERuntime%20Context%20Structure%20%28by%20ref%29))
  (Fortnite|ApplyGameplayEffectsToActor (Variables|FortMovementMode_BaseExtLogic|GetOwnerPawn) (Variables|Default|GetHolsterGameplayEffect) (Transformation|GetActorLocation (Variables|FortMovementMode_BaseExtLogic|GetOwnerPawn)) false 1 (Ability|GetAbilitySystemComponent (Variables|FortMovementMode_BaseExtLogic|GetOwnerPawn)))
  (Ability|SendGameplayEventtoActor (Variables|FortMovementMode_BaseExtLogic|GetOwnerPawn) "(TagName=\"WingSuit.End\")")
  (CallFunction|SetCameraModifier false)
  (bind _returnvalue (Utilities|Time|SetTimerbyEvent (delegate Custom|DestroyCamera) 2.0))
  (Variables|Default|SetTimerToDestroy _returnvalue)
  (bind _returnvalue_2 (Class|FortMovementModeExtBlueprintLibrary|GetRuntimeDatafromContext RuntimeContext "/Script/WingSuitRuntime.FortMovementMode_WingSuitRuntimeData"))
  (switch Utilities|FlowControl|Switch|SwitchonEWingSuitLandingMode (Class|FortMovementModeWingSuitRuntimeData|GetLandingMode _returnvalue_2)
    (:Point
      (if (> (.x (Class|FortMovementModeWingSuitRuntimeData|GetInputVector _returnvalue_2)) 0.5)
        (CallFunction|PlayAnimMontage "/WingSuit/Animation/Game/MainPlayer/AnimAssets/Wingsuit_Landing_RollToLoco_M.Wingsuit_Landing_RollToLoco_M")
        (else
          (CallFunction|PlayAnimMontage "/WingSuit/Animation/Game/MainPlayer/AnimAssets/Wingsuit_Landing_VerticalNoRoll_M.Wingsuit_Landing_VerticalNoRoll_M"))))
    (:Slide
      (CallFunction|PlayAnimMontage "/WingSuit/Animation/Game/MainPlayer/AnimAssets/Wingsuit_Landing_RollToKneeSlide_M.Wingsuit_Landing_RollToKneeSlide_M"))))

(event Custom|DestroyCamera
  (CallFunction|SetCameraModifier false true))

(fn SetCameraModifier
  (params (Enabled? :type Boolean) (DoDestroy :type Boolean))
  (bind _ownerpawn (Variables|FortMovementMode_BaseExtLogic|GetOwnerPawn))
  (if (not (Networking|HasAuthority _ownerpawn))
    (for _array_element (FortPawn|GetAllLocalViewingPlayerControllers _ownerpawn)
      (Utilities|IsValid _array_element
        (:Is%20Valid
          (bind _playercameramanager (Variables|/Script/Engine.PlayerController|GetPlayerCameraManager _array_element)
            (:then
              (bind _output_get (LocalVariables|SetPlayerCameraManager _playercameramanager))
              (Utilities|IsValid _output_get
                (:Is%20Valid
                  (if (== _array_element (Fortnite|GetFortPlayerControllerfromActor _ownerpawn))
                    (Variables|Default|SetCameraModifierCurrent (Variables|Default|GetCamera%20Modifier%20Class))
                    (else
                      (Variables|Default|SetCameraModifierCurrent (Variables|Default|GetCamera%20Modifier%20Class%20Spectator))))
                  ;[1 1 1] Main Player
                  (if Enabled?
                    (bind _returnvalue (Game|Player|AddNewCameraModifier (LocalVariables|GetPlayerCameraManager) (Variables|Default|GetCameraModifierCurrent)))
                    (CameraModifier|EnableModifier _returnvalue)
                    (else
                      (bind _returnvalue_2 (Game|Player|FindCameraModifierbyClass (LocalVariables|GetPlayerCameraManager) (Variables|Default|GetCameraModifierCurrent)))
                      (Utilities|IsValid _returnvalue_2
                        (:Is%20Valid
                          (CameraModifier|DisableModifier _returnvalue_2 false)
                          (Variables|GetDoDestroy
                            (:then
                              ;;[1 1 1] This stops the camera from being weird at start - fix this on the camera modifier if time allows
                              (Game|Player|RemoveCameraModifier (LocalVariables|GetPlayerCameraManager) _returnvalue_2))
                            (:else)))))))))
            (:else)))))))
