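; Gameplay ability that launches the player into wingsuit flight.
; A hop suspends gravity during the takeoff montage until its Boost notify spends ammo and activates the WingSuit movement mode extension, then a timed unfurl montage sends BeginFlight.
; A server-side trace timer detects head bonks during launch, cancelling movement and applying a cooldown.
; WingSuit.End or a timeout ends the ability; ending before boost still takes a charge and cooldown.
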(event Custom|EndUnfurl
  (Ability|GameplayEffect|RemoveLooseGameplayTags (Variables|Default|GetPlayerPawn) (Variables|Default|GetUnfurlTag) true)
  (Ability|EndAbility))

(event Custom|SlowGrav
  (Variables|/Script/Engine.CharacterMovementComponent|SetGravityScale 0.0 (Variables|/Script/Engine.Character|GetCharacterMovement (Variables|Default|GetPlayerPawn)))
  (Camera|PopCameraMode)
  (Camera|PushCameraMode "/WingSuit/Gameplay/WingSuit_PhysicsFlying/Camera/CameraMode_WingSuit_WarmingUp_CloseCamera_v2.CameraMode_WingSuit_WarmingUp_CloseCamera_v2_C"))

(event Custom|Unfurl
  (Ability|GameplayEffect|AddLooseGameplayTags (Variables|Default|GetPlayerPawn) (Variables|Default|GetUnfurlTag) true)
  (Ability|SendGameplayEventtoActor (Variables|Default|GetPlayerPawn) "(TagName=\"WingSuit.BeginFlight\")"))

(event Custom|EndLoop
  ;[1 1 1] end anim loop
  (Utilities|Time|ClearandInvalidateTimerbyHandle (Variables|HeadBonk|GetHeadBonkTick))
  (CallFunction|RemoveLaunchingTags)
  (Animation|Montage|MontageStop (Components|SkeletalMesh|GetAnimInstance (Variables|/Script/Engine.Character|GetMesh (CallFunction|GetAvatarasFortPlayerPawn))) 0.4 (Variables|Default|GetGroundLaunchMontage))
  (bind _retval (Class|BWeapWingSuitPhysicsFlying|GetStartingAnim (Variables|Default|GetAs%20B%20Weap%20Wing%20Suit%20Physics%20Flying)))
  (bind (_asynctaskproxy _notifytag) (Ability|Tasks|PlayMontageAndWaitNotify "WingSuitUnfurl" _retval 1.0 "")
    (:OnNotifyBegin
      (if (== _notifytag "(TagName=\"WingSuit.AnimNotify.Unfurl\")")
        (CallFunction|Unfurl)
        (elif (== _notifytag "(TagName=\"WingSuit.AnimNotify.EndUnfurl\")")
          (CallFunction|EndUnfurl))))))

(event Ability|EventActivateAbility
  (bind _returnvalue (CallFunction|GetAvatarasFortPlayerPawn))
  (bind _returnvalue_2 (Weapon|GetCurrentWeapon _returnvalue))
  (bind _polarityweapon (Default|GetPolarityWeapon))
  (bind _returnvalue_3 (Ability|GetAvatarActorfromActorInfo))
  (bind _returnvalue_4 (Utilities|Time|SetTimerbyEvent (delegate Custom|EventuallyEnd) (ScalableFloat|GetValueatLevel (Variables|Default|GetEventualEndTime))))
  (Variables|Default|SetEventualEndTimer _returnvalue_4)
  (Ability|Async|WaitGameplayEventToActor _returnvalue_3 "(TagName=\"WingSuit.End\")" true
    (:then
      (bind _asfort_pawn (Utilities|Casting|CastToFortPawn _returnvalue_3))
      (Variables|Default|SetPlayerPawn _asfort_pawn)
      (bind _asb_weap_wing_suit_physics_flying (Utilities|Casting|CastToB_Weap_WingSuit_PhysicsFlying (Weapon|GetCurrentWeapon (Variables|Default|GetPlayerPawn))))
      (Variables|Default|SetAs%20B%20Weap%20Wing%20Suit%20Physics%20Flying _asb_weap_wing_suit_physics_flying)
      (Ability|ExecuteGameplayCueOnOwner (Variables|Default|GetWarmUpTag))
      (Variables|Default|SetCleanUpEverything true)
      (Camera|PushCameraMode "/WingSuit/Gameplay/WingSuit_PhysicsFlying/Camera/CameraMode_WingSuit_WarmingUp_FarCamera.CameraMode_WingSuit_WarmingUp_FarCamera_C")
      ;[1 1 1] Initial Hop
      (Variables|/Script/Engine.CharacterMovementComponent|SetGravityScale 0.5 (Variables|/Script/Engine.Character|GetCharacterMovement (Variables|Default|GetPlayerPawn)))
      (Character|LaunchCharacter (Variables|Default|GetPlayerPawn) (Math|Vector|MakeVector 0.0 0.0 4.5e+02) true true)
      (Default|SetAirControl 0.0)
      (bind _returnvalue_5 (Utilities|Time|SetTimerbyEvent (delegate Custom|SlowGrav) 0.15))
      (Variables|Default|SetTimerHandleSlowGrav _returnvalue_5)
      (Ability|GameplayEffect|AddLooseGameplayTags (Variables|Default|GetPlayerPawn) (Variables|Default|GetLaunchingTags))
      (bind (_asynctaskproxy _notifytag) (Ability|Tasks|PlayMontageAndWaitNotify "WingSuitTakeoff" (Variables|Default|GetGroundLaunchMontage) 1.0 "Start" false)
        (:then
          (Variables|Default|SetTakeOffAnim _asynctaskproxy))
        (:OnInterrupted
          _)
        (:OnCancelled
          _)
        (:OnNotifyBegin
          (if (== _notifytag "(TagName=\"WingSuit.AnimNotify.Boost\")")
            (Variables|Default|SetCleanUpEverything false)
            (Variables|/Script/Engine.CharacterMovementComponent|SetGravityScale 1.0 (Variables|/Script/Engine.Character|GetCharacterMovement (Variables|Default|GetPlayerPawn)))
            (Camera|PopCameraMode)
            (Audio|PlaySoundatLocation (Variables|Default|GetSoundToLaunch) (Transformation|GetActorLocation (Variables|Default|GetPlayerPawn)))
            (Ability|GameplayEffect|RemoveLooseGameplayTags (Variables|Default|GetPlayerPawn) (Variables|Default|GetLaunchingTags))
            (bind _asfort_inventory_owner_interface (Utilities|Casting|CastToFortInventoryOwnerInterface (Pawn|GetController (CallFunction|GetAvatarasFortPlayerPawn))))
            (bind _returnvalue_6 (InventoryManagement|FindExistingItemforDefinition _asfort_inventory_owner_interface "/WingSuit/Gameplay/WingSuit_PhysicsFlying/WID_Proto_WingSuit_FlyingPhysics.WID_Proto_WingSuit_FlyingPhysics"))
            (bind _asfort_world_item (Utilities|Casting|CastToFortWorldItem _returnvalue_6))
            (bind _asb_weap_wing_suit_physics_flying_2 (Utilities|Casting|CastToB_Weap_WingSuit_PhysicsFlying _returnvalue_2))
            (Variables|/WingSuit/Gameplay/WingSuit_PhysicsFlying/B_Weap_WingSuit_PhysicsFlying.B_Weap_WingSuit_PhysicsFlying_C|SetDeleteOnLand (<= (Item|GetTotalAmmo _asfort_world_item) 1) _asb_weap_wing_suit_physics_flying_2)
            (FortWeapon|Ammo|ConsumeAmmo _returnvalue_2 1 true)
            (CallFunction|MMEActivate)
            (if true
              (Ability|SendGameplayEvent "(TagName=\"Event.Weapons.Fire\")")
              (Ability|SendGameplayEvent "(TagName=\"WingSuit.UseAbility\")")
              (Ability|GameplayEffect|AddLooseGameplayTags (Variables|Default|GetPlayerPawn) (Variables|Default|GetTagsBlockBassWave))
              (CallFunction|HasAbilityStarted true)
              (CallFunction|PauseUntilNextJump)
              (bind _returnvalue_7 (Utilities|Time|SetTimerbyEvent (delegate Custom|EndLoop) 1.05))
              (Variables|Default|SetTimerHandleEndLoop _returnvalue_7)
              (Game|Feedback|ClientStartCameraShake :self (Fortnite|GetFortPlayerControllerfromActor (Variables|Default|GetPlayerPawn)) :Shake (Variables|Default|GetBassJumpCamShake) :Scale 1.0 :PlaySpace "CameraLocal")
              (Ability|ApplyGameplayEffectToOwner (Variables|Default|GetFallImmunityGE))
              (Utilities|IsValid _polarityweapon
                (:Is%20Valid
                  (Class|BWeapWingSuitPhysicsFlying|TryUseDurability _polarityweapon (ScalableFloat|GetValueatLevel (Variables|Default|GetHF_AbilityDurabilityUseMultiplier))))
                (:Is%20Not%20Valid
                  _))
              (Ability|ExecuteGameplayCueOnOwner (Variables|Default|GetGCBassJumpLaunch))))))
      (if (Variables|Default|GetCleanUpEverything)
        (Ability|EndAbility)))
    (:EventReceived
      (Ability|EndAbility)
      (CallFunction|RemoveLaunchingTags)
      (Animation|Montage|MontageStop (Components|SkeletalMesh|GetAnimInstance (Variables|/Script/Engine.Character|GetMesh (CallFunction|GetAvatarasFortPlayerPawn))) 0.4 (Variables|Default|GetGroundLaunchMontage)))))

(event Custom|EventuallyEnd
  (Ability|EndAbility))

(event Custom|RemoveLaunchingTags
  (Ability|GameplayEffect|RemoveLooseGameplayTags (Variables|Default|GetPlayerPawn) (Variables|Default|GetTagsBlockBassWave) true)
  (Ability|GameplayEffect|RemoveLooseGameplayTags (Variables|Default|GetPlayerPawn) (Variables|Default|GetUnfurlTag) true))

(event Custom|PauseUntilNextJump
  ;[1 1 1] After Bass Jumping, allow next Bass Jump after X duration (separate duration for first jump to allow for later ability application)
  (if (Variables|Default|GetbIsGroundJump)
    (Utilities|Time|SetTimerbyEvent (delegate Custom|AllowNextJump) (Default|ClampMinFloat (ScalableFloat|GetValueatLevel (Variables|Default|GetGround%20TIme%20Until%20Allow%20Next%20Jump)) (ScalableFloat|GetValueasDouble (Variables|Default|GetTime%20Until%20Allow%20Next%20Jump%20Minimum))))
    (else
      (Utilities|Time|SetTimerbyEvent (delegate Custom|AllowNextJump) (Default|ClampMinFloat (ScalableFloat|GetValueatLevel (Variables|Default|GetAerial%20Time%20Until%20Allow%20Next%20Jump)) (ScalableFloat|GetValueasDouble (Variables|Default|GetTime%20Until%20Allow%20Next%20Jump%20Minimum)))))))

(event Custom|AllowNextJump)

(event Custom|MMEActivate
  (bind _returnvalue (CallFunction|GetAvatarasFortPlayerPawn))
  (Ability|SendGameplayEvent (Variables|Default|GetEventTag_RemoveBalloons)) ; remove all balloons from the player \nsince they are doing a MME action\nand should be consistent with other features\n(like Doom Gauntlets - FORT-766407)
  (if (GameplayTags|HasMatchingGameplayTag (CallFunction|GetAvatarasFortPlayerPawn) "(TagName=\"MovementMode.Flying.WingSuit\")")
    _
    (elif (Variables|Debug|GetUse%20Native%20MME?)
      ;[0.2969 0.6068 1 0.5] If we are using native MME (Movement Mode Extention), Activate Wingsuit MME
      ; Else; Clear MME
      (bind _outruntimedata (Class|FortMovementModeExtLogicWingSuit|TryActivateWingSuitMME _returnvalue "/WingSuit/Gameplay/WingSuit_PhysicsFlying/Native/FMM_ExtLogicWingSuit.FMM_ExtLogicWingSuit_C" (Utilities|Struct|MakeFortMovementModeWingSuitCreationData true)))
      (Variables|Default|SetMME _outruntimedata)
      (else
        (bind _returnvalue_2 (Class|FortMovementModeExtBlueprintLibrary|ActivateRegisteredMovementModeExtension 0 _returnvalue))
        (Variables|Default|SetMME _returnvalue_2))))
  (if (and (ScalableFloat|GetValueasBool (Variables|HeadBonk|GetHeadBonkIsEnabled)) (Networking|HasAuthority (Variables|Default|GetPlayerPawn)))
    (Variables|HeadBonk|SetHeadBonkStartTime -1.0)
    (bind _returnvalue_3 (Utilities|Time|SetTimerbyEvent (delegate Custom|CheckForHeadBonk) (ScalableFloat|GetValueatLevel (Variables|HeadBonk|GetBonkTimerInterval)) true true))
    (Variables|HeadBonk|SetHeadBonkTick _returnvalue_3)))

(event Ability|EventOnEndAbility
  (params (bWasCancelled :type Boolean))
  (bind _headbonkrefundcd (Variables|HeadBonk|GetHeadBonkRefundCD))
  (bind _returnvalue (ScalableFloat|GetValueatLevel _headbonkrefundcd))
  (Utilities|Time|ClearandInvalidateTimerbyHandle (Variables|Default|GetEventualEndTimer))
  (Utilities|Time|ClearandInvalidateTimerbyHandle (Variables|HeadBonk|GetHeadBonkTick))
  (Utilities|Time|ClearandInvalidateTimerbyHandle (Variables|Default|GetTimerHandleSlowGrav))
  (Variables|/Script/Engine.CharacterMovementComponent|SetGravityScale 1.0 (Variables|/Script/Engine.Character|GetCharacterMovement (Variables|Default|GetPlayerPawn)))
  (Utilities|Time|ClearandInvalidateTimerbyHandle (Variables|Default|GetTimerHandleEndLoop))
  (CallFunction|HasAbilityStarted false)
  (CallFunction|RemoveLaunchingTags)
  (if (Variables|Default|GetCleanUpEverything)
    (Camera|PopCameraMode)
    (Ability|GameplayEffect|RemoveLooseGameplayTags (Variables|Default|GetPlayerPawn) (Variables|Default|GetLaunchingTags))
    (Variables|/Script/Engine.CharacterMovementComponent|SetGravityScale 1.0 (Variables|/Script/Engine.Character|GetCharacterMovement (Variables|Default|GetPlayerPawn)))
    ;[1 1 1] Put on shorter Cd and take a charge
    (FortWeapon|Ammo|ConsumeAmmo (Weapon|GetCurrentWeapon (CallFunction|GetAvatarasFortPlayerPawn)) 1 true)
    (bind _returnvalue_2 (Ability|GameplayEffect|SetDuration (Ability|MakeOutgoingGameplayEffectSpec "/WingSuit/Gameplay/WingSuit_PhysicsFlying/GE_WingSuit_PhysicsFlying_Cooldown.GE_WingSuit_PhysicsFlying_Cooldown_C") _returnvalue))
    (Ability|ApplyGameplayEffectSpecToOwner _returnvalue_2)
    (bind _returnvalue_3 (Ability|GameplayEffect|SetDuration (Ability|MakeOutgoingGameplayEffectSpec "/WingSuit/Gameplay/WingSuit_PhysicsFlying/GE_WingSuit_PhysicsFlying_Cooldown_Visuals.GE_WingSuit_PhysicsFlying_Cooldown_Visuals_C") _returnvalue))
    (Ability|ApplyGameplayEffectSpecToOwner _returnvalue_3)
    (Class|BWeapWingSuitPhysicsFlying|CheckRemoveWeapononEmpty (Variables|Default|GetAs%20B%20Weap%20Wing%20Suit%20Physics%20Flying))))

(event Custom|CheckForHeadBonk
  (bind _headbonkrefundcd (Variables|HeadBonk|GetHeadBonkRefundCD))
  (bind _returnvalue (ScalableFloat|GetValueatLevel _headbonkrefundcd))
  (bind (_outhit _returnvalue_2) (Collision|LineTraceByChannel (+ (Transformation|GetActorLocation (Variables|Default|GetPlayerPawn)) (Math|Vector|MakeVector 0.0 0.0 (ScalableFloat|GetValueatLevel (Variables|HeadBonk|GetBonkTraceDepth)))) (+ (Transformation|GetActorLocation (Variables|Default|GetPlayerPawn)) (Math|Vector|MakeVector 0.0 0.0 (ScalableFloat|GetValueatLevel (Variables|HeadBonk|GetBonkTraceHeight)))) "TraceTypeQuery10"))
  (if (ScalableFloat|GetValueasBool (Variables|HeadBonk|GetDoHeadBonkAdvancedCheck))
    (bind _shouldcancelmovement (CallFunction|ProcessHeadBonkHit _outhit))
    (if _shouldcancelmovement
      _)
    (else
      _))
  (if _returnvalue_2
    (Utilities|Casting|CastToB_Weap_WingSuit_PhysicsFlying (Weapon|GetCurrentWeapon (CallFunction|GetAvatarasFortPlayerPawn)))
    (Utilities|Time|ClearandInvalidateTimerbyHandle (Variables|HeadBonk|GetHeadBonkTick))
    (bind _returnvalue_3 (Ability|GameplayEffect|SetDuration (Ability|MakeOutgoingGameplayEffectSpec "/WingSuit/Gameplay/WingSuit_PhysicsFlying/GE_WingSuit_PhysicsFlying_Cooldown.GE_WingSuit_PhysicsFlying_Cooldown_C") _returnvalue))
    (Ability|ApplyGameplayEffectSpecToOwner _returnvalue_3)
    (bind _returnvalue_4 (Ability|GameplayEffect|SetDuration (Ability|MakeOutgoingGameplayEffectSpec "/WingSuit/Gameplay/WingSuit_PhysicsFlying/GE_WingSuit_PhysicsFlying_Cooldown_Visuals.GE_WingSuit_PhysicsFlying_Cooldown_Visuals_C") _returnvalue))
    (Ability|ApplyGameplayEffectSpecToOwner _returnvalue_4)
    (CallFunction|OnHeadBonk)
    (Ability|EndAbility)))

(event Custom|OnHeadBonkClient
  (Class|FortMovementModeExtBlueprintLibrary|EndMovementModeExtensiononActor (Variables|Default|GetMME) (CallFunction|GetAvatarasFortPlayerPawn))
  (Animation|Montage|MontageStop (Components|SkeletalMesh|GetAnimInstance (Variables|/Script/Engine.Character|GetMesh (CallFunction|GetAvatarasFortPlayerPawn))) 0.0 (Variables|Default|GetGroundLaunchMontage)))

(event Custom|OnHeadBonk
  (CallFunction|OnHeadBonkClient)
  (Class|FortMovementModeExtBlueprintLibrary|EndMovementModeExtensiononActor (Variables|Default|GetMME) (CallFunction|GetAvatarasFortPlayerPawn))
  (Animation|Montage|MontageStop (Components|SkeletalMesh|GetAnimInstance (Variables|/Script/Engine.Character|GetMesh (CallFunction|GetAvatarasFortPlayerPawn))) 0.0 (Variables|Default|GetGroundLaunchMontage)))

(fn BassJumpBlastThrough
  (bind (_returnvalue_x _returnvalue_y) (Math|Vector|BreakVector (Transformation|GetActorLocation (Variables|Default|GetPlayerPawn))))
  (bind (_returnvalue_x_2 _returnvalue_y_2 _returnvalue_z) (Math|Vector|BreakVector (Transformation|GetActorLocation (Variables|Default|GetPlayerPawn))))
  (if (> (ScalableFloat|GetValueatLevel (Variables|Default|GetBassJumpShouldDestroyAbove)) 0.0)
    ;[1 1 1] Currently just applies vertically, investigate matching it to the player's path
    (bind _outhits (Collision|MultiCapsuleTraceForObjects :Start (Math|Vector|MakeVector _returnvalue_x_2 _returnvalue_y_2 (+ _returnvalue_z (ScalableFloat|GetValueatLevel (Variables|Default|GetBassJumpDestroyDistanceZOffset)))) :End (Math|Vector|MakeVector _returnvalue_x _returnvalue_y (+ _returnvalue_z (ScalableFloat|GetValueatLevel (Variables|Default|GetBassJumpDestroyDistanceZOffset)))) :Radius 1.6e+02 :HalfHeight 3.5e+02 :ObjectTypes (Variables|Default|GetObjectTypesToDestroy) :bTraceComplex false :DrawDebugType "None" :bIgnoreSelf true :TraceColor "(R=1.000000,G=0.000000,B=0.000000,A=1.000000)" :TraceHitColor "(R=0.000000,G=1.000000,B=0.000000,A=1.000000)" :DrawTime 8.0))
    (for _array_element _outhits
      (if (Variables|/Script/Engine.Actor|GetbCanBeDamaged (.HitActor (Collision|BreakHitResult _array_element)))
        (bind _bdidactorhavetags (Class|AthenaQuestFunctionLibrary|AthenaCheckActorHasTags (.HitActor (Collision|BreakHitResult _array_element)) "(GameplayTags=((TagName=\"Deimos.Spawner\")))"))
        (if (not _bdidactorhavetags)
          (Utilities|Casting|CastToBuildingTrap (.HitActor (Collision|BreakHitResult _array_element))
            (:CastFailed
              (GameplayEffects|ApplyGameplayEffectToTarget :self (Ability|GetAbilitySystemComponent (Variables|Default|GetPlayerPawn)) :GameplayEffectClass "/Game/Athena/Items/Consumables/ShockwaveGrenade/GE_ShockGrenade_Damage.GE_ShockGrenade_Damage_C" :Target (Ability|GetAbilitySystemComponent (.HitActor (Collision|BreakHitResult _array_element))) :Level 0.0))))))))

(fn CalculateLaunchDirection
  (bind _playerpawn (Variables|Default|GetPlayerPawn)
    (:then
      (bind _output_get (LocalVariables|SetPlayerPawnVar _playerpawn))
      (bind _asfort_player_pawn (Utilities|Casting|CastToFortPlayerPawn _output_get))
      (LocalVariables|SetFortPlayerPawnVar _asfort_player_pawn)
      (bind (_returnvalue_location _returnvalue_rotation) (Math|Transform|BreakTransform (Targeting|GetTargetingTransform _playerpawn "Camera" 0)))
      (return (Math|Vector|GetForwardVector _returnvalue_rotation)))
    (:else)))

(fn CanBassJump
  ;[1 1 1] If player can multi-jump skip other checks
  (if (ScalableFloat|GetValueasBool (Variables|Default|GetbAllowMultiBassJump))
    (return true)
    (else
      (return (not (GameplayTags|HasMatchingGameplayTag (Variables|Default|GetPlayerPawn) (Variables|Default|GetTagPlayerBassJumpActive)))))))

(fn GetPolarityWeapon
  (bind _asb_weap_wing_suit_physics_flying (Utilities|Casting|CastToB_Weap_WingSuit_PhysicsFlying (Ability|GetCurrentSourceWeapon))
    (:then
      (return _asb_weap_wing_suit_physics_flying))
    (:CastFailed
      (return 0))))

(fn HasAbilityStarted
  (params (AbilityStart :type Boolean))
  (bind _asb_weap_wing_suit_physics_flying (Utilities|Casting|CastToB_Weap_WingSuit_PhysicsFlying (Weapon|GetCurrentWeapon (Variables|Default|GetPlayerPawn))))
  (Variables|/WingSuit/Gameplay/WingSuit_PhysicsFlying/B_Weap_WingSuit_PhysicsFlying.B_Weap_WingSuit_PhysicsFlying_C|SetLaunchActivated AbilityStart _asb_weap_wing_suit_physics_flying))

(fn IsHitAgainstWall
  (params (HitResult :type Hit%20Result%20Structure))
  (bind (_bblockinghit _binitialoverlap _time _distance _location _impactpoint _normal _impactnormal) (Collision|BreakHitResult HitResult))
  (return (>= (Math|Trig|Acos(Degrees) (Math|Vector|DotProduct _impactnormal "0.000000,0.000000,1.000000")) 5e+01)))

(fn CanActivateAbility
  (params (ActorInfo :type Gameplay%20Ability%20Actor%20Info%20Structure) (Handle :type Gameplay%20Ability%20Spec%20Handle%20Structure))
  (return 0 (>= (ScalableFloat|GetValueatLevel (Variables|Default|GetHF_CanActivateAbility)) 1.0)))

(fn ProcessHeadBonkHit
  (params (Hit%20Result :type Hit%20Result%20Structure))
  (bind (_bblockinghit _binitialoverlap _time _distance _location _impactpoint _normal) (Collision|BreakHitResult Hit%20Result))
  (bind _returnvalue (Utilities|Time|GetGameTimeinSeconds))
  (if _bblockinghit
    (if (<= (Math|Trig|Acos(Degrees) (Math|Vector|DotProduct _normal (Math|Vector|VectorDown))) (ScalableFloat|GetValueatLevel (Variables|HeadBonk|GetHeadBonkAngleRequirement)))
      (if (< (Variables|HeadBonk|GetHeadBonkStartTime) 0)
        (Variables|HeadBonk|SetHeadBonkStartTime _returnvalue)
        (else
          _))
      (return (>= (- _returnvalue (Variables|HeadBonk|GetHeadBonkStartTime)) (ScalableFloat|GetValueatLevel (Variables|HeadBonk|GetHeadBonkTimeRequirement))))
      (else
        _))
    (else
      _))
  (Variables|HeadBonk|SetHeadBonkStartTime -1.0)
  (return false))

(fn SetAirControl
  (params (AirControl :type Float%20%28single-precision%29))
  (if (!= (Variables|Default|GetCurrentAirControl) AirControl)
    (Variables|Default|SetCurrentAirControl AirControl)
    (Ability|RemoveGameplayEffectFromOwnerWithHandle (Variables|Default|GetGE_AirControl))
    (bind _returnvalue (Ability|GameplayEffect|AssignTagSetbyCallerMagnitude (GameplayEffects|MakeOutgoingSpec :self (Ability|GetFortAbilitySystemComponent (Variables|Default|GetPlayerPawn)) :GameplayEffectClass (Variables|Default|GetGE_AirControl_Class) :Level 0.0) "(TagName=\"SetByCaller.Magnitude\")" AirControl))
    (bind _returnvalue_2 (Ability|ApplyGameplayEffectSpecToOwner _returnvalue))
    (Variables|Default|SetGE_AirControl _returnvalue_2)))
